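# NE_GateVFX is a Niagara system for a portal gate, pairing the visuals with a persistent portal loop sound played by the Center emitter.
# Center draws large glow sprites, two MeshRibbon emitters layer randomly rotated, fading meshes with animated material parameters, and Minimal bursts large meshes.
# Three GPU sprite emitters add ground smoke, ring and side smoke, and dense sparks spawned on a System.SparkRadius cylinder, swirled by curl noise, vortex and point forces.
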NiagaraSystem /PistolCanary/Assets/FX/Portal/NE_GateVFX
dependencies {
 usedDataInterfaces [
  &NiagaraDataInterfaceCurve
  &NiagaraDataInterfaceCurlNoise
  &NiagaraDataInterfaceColorCurve
  &NiagaraDataInterfaceAudioPlayer
 ]
 usedDynamicInputs []
 usedModules [
  &self:SystemSpawnScript.SystemScriptSource.SystemScriptGraph.NiagaraNodeAssignment_0.SetVariables_#1
  &SystemState
 ]
}
emitters {
 Center {
  emitterSpawnScript [
   &InitializeAudioPlayer(Audio Player=DataInterface() {propertyValues {
      Attenuation=null
      Concurrency=null
      MaxPlaysPerTick=10
      SoundToPlay=/Hermes/Props/ClayTile/ClayTile_DragonGate_B/Audio/SC_DemonArena_Portal_Loop
      bAllowLoopingOneShotSounds=false
      bLimitPlaysPerTick=true
      bOnlyActiveDuringGameplay=true
      bStopWhenComponentIsDestroyed=true
      ConfigurationUserParameter {parameter {name=None, typeDefHandle {registeredTypeIndex=1}, varData []}}
      ParameterNames []
     }})
   /Niagara/Modules/Audio/PlayPersistentAudio()
  ]
  emitterUpdateScript [
   &EmitterState(Life Cycle Mode=::NewEnumerator1, Loop Duration=1, UseLoopDelay=false)
   &SpawnRate(SpawnRate=0.75, Use Spawn Probability=false)
   /Niagara/Modules/Audio/UpdatePersistentAudio()
  ]
  particleSpawnScript [
   /Niagara/Modules/Spawn/Initialization/InitializeParticle(
    Color=<rgba 0.0727 0.7686 1 1>
    Lifetime=8
    Module.Write Color=true
    Module.Write Lifetime=true
    Module.Write Mass=true
    Module.Write Material Random=true
    Module.Write Position=false
    Module.Write Scale=false
    Module.Write SpriteRotation=true
    Module.Write SpriteSize=true
    Position=&SimulationPosition()
    Sprite Rotation=&UniformRangedFloat(Minimum=0, Maximum=360)
    Sprite Size=<550 550>
   ) {enabled=false}
   /Niagara/Modules/Spawn/Initialization/V2/InitializeParticle(
    Color=<rgba 2 0.5521 0.9382 1>
    Lifetime=8
    Mass Mode=::NewEnumerator1
    Sprite Rotation Angle=&UniformRangedFloat(Minimum=0, Maximum=360)
    Sprite Rotation Mode=::NewEnumerator0
    Sprite Size=<650 650>
    Sprite Size Mode=::NewEnumerator4
    UsePositionOffset=false
    Write Material Random=true
   ) {module=InitializeParticle001}
   /SetVariables(
    Particles.Position=&Add_Vector(A=Linked() {linkedVariable {name=Particles.Position, type=NiagaraPosition}}, B=<0 0 0>)
    Particles.Random=&UniformRangedFloat(Minimum=-0.1, Maximum=0.9)
   ) {isSetParameters=true}
  ]
  particleUpdateScript [
   &ParticleState()
   &ScaleSpriteSize(Scale Sprite Size Mode=::NewEnumerator0, Uniform Scale Factor=4)
   &SolveForcesAndVelocity(Clamp Velocity=false, Limit Acceleration=false)
   &DynamicMaterialParameters(
    Index 0 Param 1=&FloatFromCurve(FloatCurve=DataInterface() {propertyValues {Curve {keys [
         {arriveTangent=1.1486, interpMode=RCIM_Cubic, leaveTangent=1.1486, tangentMode=RCTM_User, time=0, value=0}
         {arriveTangent=-9.5651, interpMode=RCIM_Cubic, leaveTangent=-9.5651, tangentMode=RCTM_User, time=0.6446, value=-2.4156}
         {arriveTangent=-1.1502, interpMode=RCIM_Cubic, leaveTangent=-1.1502, tangentMode=RCTM_User, time=1, value=-3.0414}
        ]}}})
    Index 0 Param 2=&FloatFromCurve(FloatCurve=DataInterface() {propertyValues {Curve {keys [
         {arriveTangent=-10.815, interpMode=RCIM_Cubic, leaveTangent=-10.815, tangentMode=RCTM_User, time=0.0013, value=0.9969}
         {arriveTangent=-0.881, interpMode=RCIM_Cubic, leaveTangent=-0.881, time=0.3186, value=0.0305}
         {arriveTangent=2.4595, interpMode=RCIM_Cubic, leaveTangent=2.4595, time=0.758, value=0.3302}
         {arriveTangent=3.5097, interpMode=RCIM_Cubic, leaveTangent=3.5097, tangentMode=RCTM_User, time=1, value=1.7064}
        ]}}})
    Index 0 Param 4=&Add_Float(
     A=$Particles.Random
     B=&FloatFromCurve(
      FloatCurve=DataInterface() {propertyValues {Curve {keys [{interpMode=RCIM_Cubic, tangentMode=RCTM_User, time=0, value=0}, {interpMode=RCIM_Cubic, tangentMode=RCTM_User, time=1, value=1}]}}}
      Scale Curve=2
     )
    )
    Module.Param0WriteEnabled=true
    Module.Param1WriteEnabled=true
    Module.Param2WriteEnabled=true
    Module.Param3WriteEnabled=true
   )
   &ScaleColor(
    Scale Mode=::NewEnumerator1
    Scale RGBA=&Vector4FromCurve(Vector4Curve=DataInterface() {propertyValues {
       AlphaCurve {keys [
         <at 0.0037 0 RCIM_Cubic>
         {arriveTangent=2.9787, interpMode=RCIM_Cubic, leaveTangent=2.9787, time=0.1552, value=0.2595}
         {arriveTangent=1.0157, interpMode=RCIM_Cubic, leaveTangent=1.0157, time=0.2714, value=0.7974}
         {arriveTangent=-1.0205, interpMode=RCIM_Cubic, leaveTangent=-1.0205, time=0.7097, value=0.8228}
         {arriveTangent=-2.8344, interpMode=RCIM_Cubic, leaveTangent=-2.8344, time=0.8433, value=0.2138}
         <at 1 0 RCIM_Cubic>
        ]}
       BlueCurve {keys [<at 0.2703 1>,<at 0.8712 0>]}
       GreenCurve {keys [<at 0.2703 1>,<at 0.8712 0>]}
       RedCurve {keys [<at 0.2703 1>,<at 0.8712 0>]}
      }})
    ScaleRGBA=true
   )
   &CameraOffset() {enabled=false}
   &self:MatchScale() {module=MatchScale}
  ]
  rendererClasses [&NiagaraSpriteRendererProperties]
  renderers [{rendererClass=&NiagaraSpriteRendererProperties}]
 }
 MeshRibbon {
  emitterSpawnScript [/SetVariables(Emitter.MeshRibbonData=DataInterface() {propertyValues {
      bExposeCurve=true
      AlphaCurve {keys []}
      BlueCurve {keys [<at 0 1>,<at 1 1>]}
      GreenCurve {keys [<at 0 -0.5 RCIM_Cubic>, {arriveTangent=0.0331, interpMode=RCIM_Cubic, leaveTangent=0.0331, tangentMode=RCTM_User, time=1, value=-1}]}
      RedCurve {keys [<at 0 1 RCIM_Cubic>,<at 1 2 RCIM_Cubic>]}
     }}) {isSetParameters=true, module=SetVariables_#2}]
  emitterUpdateScript [
   &EmitterState()
   &SpawnBurst_Instantaneous(Spawn Count=12, Use Spawn Probability=false, UseLoopCountLimit=false) {enabled=false}
   &SpawnRate(SpawnRate=2, Use Spawn Probability=false)
  ]
  particleSpawnScript [
   &InitializeParticle(
    Color=<rgba 5 5 5 1>
    Lifetime=&RandomRangeFloat(Minimum=1.7945, Maximum=3.5428)
    Mesh Scale=&Multiply_VectorByFloat(Vector=<1 1 2.5>, Float=7)
    Mesh Scale Mode=::NewEnumerator4
    UsePositionOffset=false
    Write Material Random=true
   )
   &InitialMeshOrientation(
    Mesh Orientation Mode=::NewEnumerator2
    Orientation Vector=<0 0 -1>
    Rotation=&RandomRangeVector(Minimum=<0 0 0>, Maximum=<0 0 350>)
    Use Rotation Vector=true
   )
  ]
  particleUpdateScript [
   &ParticleState()
   &ScaleMeshSize(Scale Factor=&VectorFromFloat(Value=&FloatFromCurve(FloatCurve=DataInterface() {propertyValues {Curve {keys [<at 0 1 RCIM_Cubic>,<at 1 0.9 RCIM_Cubic>]}}})))
   &DynamicMaterialParameters(
    Index 0 Param 1=&FloatFromCurve(FloatCurve=DataInterface() {propertyValues {Curve {keys [
         {interpMode=RCIM_Cubic, tangentMode=RCTM_User, time=0, value=0.5}
         {arriveTangent=2.7589, interpMode=RCIM_Cubic, leaveTangent=2.7589, tangentMode=RCTM_User, time=1, value=2}
        ]}}})
    Index 0 Param 2=&FloatFromCurve(FloatCurve=DataInterface() {propertyValues {Curve {keys [{interpMode=RCIM_Cubic, tangentMode=RCTM_User, time=0, value=0}, {interpMode=RCIM_Cubic, tangentMode=RCTM_User, time=1, value=1}]}}})
    Index 0 Param 3=&RandomRangeFloat(Minimum=3.0512, Maximum=5.4)
    Index 0 Param 4=&Add_Float(A=&RandomRangeFloat(Minimum=16, Maximum=40), B=0)
    Module.Index0Param0Write=true
    Module.Index0Param1Write=true
    Module.Index0Param2Write=true
    Module.Index0Param3Write=true
   )
   &ScaleColor(
    Linear Color Curve=DataInterface() {propertyValues {
      AlphaCurve {keys [<at 0 0 RCIM_Cubic>,<at 0.2932 3 RCIM_Cubic>,<at 1 0 RCIM_Cubic>]}
      BlueCurve {keys [<at 0 0 RCIM_Cubic>,<at 0.2925 3 RCIM_Cubic>,<at 1 0 RCIM_Cubic>]}
      GreenCurve {keys [<at 0 0 RCIM_Cubic>,<at 0.2925 3 RCIM_Cubic>,<at 1 0 RCIM_Cubic>]}
      RedCurve {keys [<at 0 0 RCIM_Cubic>,<at 0.2925 3 RCIM_Cubic>,<at 1 0 RCIM_Cubic>]}
     }}
    Scale Mode=::NewEnumerator2
   )
   &CameraOffset(Camera Offset Amount=50)
  ]
  rendererClasses [&NiagaraMeshRendererProperties]
  renderers [{rendererClass=&NiagaraMeshRendererProperties, rendererIndex=0}]
 }
 MeshRibbon001 {
  emitterSpawnScript [/SetVariables(Emitter.MeshRibbonData=DataInterface() {propertyValues {
      bExposeCurve=true
      AlphaCurve {keys []}
      BlueCurve {keys [<at 0 1>,<at 1 1>]}
      GreenCurve {keys [<at 0 -0.5 RCIM_Cubic>, {arriveTangent=0.0331, interpMode=RCIM_Cubic, leaveTangent=0.0331, tangentMode=RCTM_User, time=1, value=-1}]}
      RedCurve {keys [<at 0 1 RCIM_Cubic>,<at 1 2 RCIM_Cubic>]}
     }}) {isSetParameters=true, module=SetVariables_#2}]
  emitterUpdateScript [
   &EmitterState()
   &SpawnBurst_Instantaneous(Spawn Count=12, Use Spawn Probability=false, UseLoopCountLimit=false) {enabled=false}
   &SpawnRate(SpawnRate=2, Use Spawn Probability=false)
  ]
  particleSpawnScript [
   &InitializeParticle(
    Color=<rgba 1 1 1 11>
    Lifetime=&RandomRangeFloat(Minimum=5.5268, Maximum=8.8377)
    Mesh Scale=&Multiply_VectorByFloat(Vector=<1 -1 1>, Float=&RandomRangeFloat(Minimum=7.9909, Maximum=12.017))
    Mesh Scale Mode=::NewEnumerator4
    UsePositionOffset=false
    Write Material Random=true
   )
   &InitialMeshOrientation(
    Mesh Orientation Mode=::NewEnumerator2
    Orientation Vector=<0 0 -1>
    Rotation=&RandomRangeVector(Minimum=<0 0 0>, Maximum=<0 0 350>)
    Use Rotation Vector=true
   )
  ]
  particleUpdateScript [
   &ParticleState()
   &ScaleMeshSize(Scale Factor=&VectorFromFloat(Value=&FloatFromCurve(FloatCurve=DataInterface() {propertyValues {Curve {keys [{arriveTangent=0.8366, interpMode=RCIM_Cubic, leaveTangent=0.8366, tangentMode=RCTM_User, time=0, value=0.5}, <at 1 0.75 RCIM_Cubic>]}}})))
   &DynamicMaterialParameters(
    Index 0 Param 1=&FloatFromCurve(FloatCurve=DataInterface() {propertyValues {Curve {keys [{tangentMode=RCTM_User, time=0, value=-1}, {tangentMode=RCTM_User, time=1, value=0}]}}})
    Index 0 Param 2=&FloatFromCurve(FloatCurve=DataInterface() {propertyValues {Curve {keys [{tangentMode=RCTM_User, time=0, value=0}, {tangentMode=RCTM_User, time=1, value=1}]}}})
    Index 0 Param 3=&RandomRangeFloat(Minimum=2.5103, Maximum=7.4856)
    Index 0 Param 4=&Add_Float(A=&RandomRangeFloat(Minimum=25.1137, Maximum=44.0696), B=0)
    Module.Index0Param0Write=true
    Module.Index0Param1Write=true
    Module.Index0Param2Write=true
    Module.Index0Param3Write=true
   )
   &ScaleColor(
    Linear Color Curve=DataInterface() {propertyValues {
      AlphaCurve {keys [<at 0 0 RCIM_Cubic>,<at 0.6722 5 RCIM_Cubic>,<at 1 0 RCIM_Cubic>]}
      BlueCurve {keys [<at 0 0 RCIM_Cubic>,<at 0.6722 5 RCIM_Cubic>,<at 1 0 RCIM_Cubic>]}
      GreenCurve {keys [<at 0 0 RCIM_Cubic>,<at 0.6722 5 RCIM_Cubic>,<at 1 0 RCIM_Cubic>]}
      RedCurve {keys [<at 0 0 RCIM_Cubic>,<at 0.6722 5 RCIM_Cubic>,<at 1 0 RCIM_Cubic>]}
     }}
    Scale Mode=::NewEnumerator2
   )
   &CameraOffset(Camera Offset Amount=50)
  ]
  rendererClasses [&NiagaraMeshRendererProperties]
  renderers [{rendererClass=&NiagaraMeshRendererProperties, rendererIndex=0}]
 }
 Minimal {
  emitterUpdateScript [
   &EmitterState()
   &SpawnBurst_Instantaneous(Use Spawn Probability=false, UseLoopCountLimit=false)
   &SpawnBurst_Instantaneous(Spawn Time=0.8833, Use Spawn Probability=false, UseLoopCountLimit=false) {module=SpawnBurst_Instantaneous001}
  ]
  particleSpawnScript [&InitializeParticle(
    Color Mode=::NewEnumerator0
    Lifetime=5
    Mesh Scale=<25 25 25>
    Mesh Scale Mode=::NewEnumerator4
    UsePositionOffset=true
    Write Material Random=true
   )]
  particleUpdateScript [
   &ParticleState()
   &ScaleMeshSize()
   &self:MatchScale() {enabled=false, module=MatchScale}
  ]
  rendererClasses [&NiagaraMeshRendererProperties]
  renderers [{rendererClass=&NiagaraMeshRendererProperties, rendererIndex=0}]
 }
 SmokGround {
  simTarget=GPUComputeSim
  emitterUpdateScript [
   &EmitterState()
   &SpawnRate(SpawnRate=1.5, Use Spawn Probability=false)
  ]
  particleSpawnScript [
   &InitializeParticle(
    Color=<rgba 1 0 0.3333 0.8>
    Lifetime=&RandomRangeFloat(Minimum=8.5812, Maximum=17.5843)
    Position Offset=<0 0 -869.0536>
    Sprite Size Mode=::NewEnumerator3
    Sprite UV Mode=::NewEnumerator3
    Uniform Sprite Size=&RandomRangeFloat(Minimum=1, Maximum=2)
    UsePositionOffset=true
    Write Material Random=true
   )
   &ShapeLocation(
    Cylinder Radius=1351.9327
    Hemicircle Internal Cap=false
    Hemicircle X=true
    MaskSpawnGroup=false
    Override Seed=false
    Shape Primitive=::NewEnumerator1
    Surface Only=true
    Use Endcaps In Surface Only Mode=false
    UseOwnerScale=true
   )
   /SetVariables(Particles.SpriteRotation:NiagaraFloat=&RandomRangeFloat(Minimum=-0.2293, Maximum=0.3106)) {isSetParameters=true, module=SetVariables_#3}
   &AddVelocity(Velocity=$StackContext.ShapeLocation.ShapeVector, Velocity Speed Scale=0.1)
   /SetVariables(Particles.SpriteSize=&Multiply_Vector2D_ByFloat(
     Float=&Lerp_Float(A=107.586, B=190.6381, Alpha=&MakeFloatFromInt(Int=$Particles.SpawnGroup))
     Vector2D=$Particles.SpriteSize
    )) {isSetParameters=true, module=SetVariables_#4}
   &ApplyInitialForces(Positional Force Warmup Time=5)
  ]
  particleUpdateScript [
   &Drag(Drag=0.1, Use Linear Drag=true, Use Rotational Drag=true)
   &VortexForce(
    Origin Pull Amount=76.4139
    UseOriginOffset=false
    Vortex Axis=<1 0 0>
    Vortex Force Amount=&Lerp_Float(
     A=&RandomRangeFloat(Minimum=93.3802, Maximum=178.6482)
     B=0
     Alpha=&MakeFloatFromInt(Int=$Particles.SpawnGroup)
    )
   ) {enabled=false}
   &GravityForce(Gravity=&Lerp_Vector(A=<0 0 -85.0702>, B=<0 0 -226.7489>, Alpha=&MakeFloatFromInt(Int=$Particles.SpawnGroup))) {enabled=false}
   &AccelerationForce(Acceleration=<200 0 0>, Coordinate Space=::Local) {enabled=false}
   &SolveForcesAndVelocity(Clamp Velocity=false, Limit Acceleration=false)
   &ParticleState()
   &ScaleSpriteSize(
    Uniform Curve Scale=5
    Uniform Curve Sprite Scale=DataInterface() {propertyValues {Curve {keys [{arriveTangent=1.2245, interpMode=RCIM_Cubic, leaveTangent=1.2245, tangentMode=RCTM_User, time=0, value=1.5}, <at 1 2 RCIM_Cubic>]}}}
   )
   /SetVariables(Particles.SubImageIndex=&FloatFromCurve(FloatCurve=DataInterface() {propertyValues {Curve {keys [<at 0 0>,<at 1 1>]}}}, Scale Curve=16)) {isSetParameters=true, module=SetVariables_#5}
   &ScaleColor(
    Scale Alpha=&FloatFromCurve(FloatCurve=DataInterface() {propertyValues {Curve {keys [
         <at 0 0 RCIM_Cubic>
         {arriveTangent=1, interpMode=RCIM_Cubic, leaveTangent=1, time=0.2, value=0.775}
         {arriveTangent=-0.9688, interpMode=RCIM_Cubic, leaveTangent=-0.9688, time=0.7, value=0.7}
         <at 1 0 RCIM_Cubic>
        ]}}})
    Scale RGB=&VectorFromFloat(Value=&FloatFromCurve(FloatCurve=DataInterface() {propertyValues {Curve {keys [{arriveTangent=-0.5678, interpMode=RCIM_Cubic, leaveTangent=-0.5678, tangentMode=RCTM_User, time=0, value=0.1}, <at 0.5 0 RCIM_Cubic>]}}}))
    ScaleA=true
    ScaleRGB=true
   )
   &DynamicMaterialParameters(
    Index 0 Param 4=&FloatFromCurve(FloatCurve=DataInterface() {propertyValues {Curve {keys [<at 0 1>,<at 1 0>]}}}, Scale Curve=2)
    Module.Index0Param0Write=true
    Module.Index0Param1Write=true
    Module.Index0Param2Write=true
    Module.Index0Param3Write=true
   )
   &CameraOffset(Camera Offset Amount=&FloatFromCurve(FloatCurve=DataInterface() {propertyValues {Curve {keys [<at 0 -50>,<at 1 -10>]}}}))
   &self:MatchScale() {module=MatchScale}
  ]
  rendererClasses [&NiagaraSpriteRendererProperties]
  renderers [{rendererClass=&NiagaraSpriteRendererProperties}]
 }
 Smoke {
  simTarget=GPUComputeSim
  emitterUpdateScript [
   &EmitterState()
   &SpawnRate(SpawnRate=5, Use Spawn Probability=false)
   &SpawnRate(SpawnGroup=1, SpawnRate=2, Use Spawn Probability=false) {module=SpawnRate001}
  ]
  particleSpawnScript [
   &InitializeParticle(
    Color=<rgba 1 0 0.3333 1>
    Lifetime=&RandomRangeFloat(Minimum=3.1968, Maximum=7.0176)
    Sprite Size Mode=::NewEnumerator3
    Sprite UV Mode=::NewEnumerator3
    Uniform Sprite Size=&RandomRangeFloat(Minimum=0.5, Maximum=1)
    UsePositionOffset=false
    Write Material Random=true
   )
   &ShapeLocation(
    MaskSpawnGroup=true
    Override Seed=false
    Ring Radius=970.7466
    Rotation Mode=::NewEnumerator2
    Shape Primitive=::NewEnumerator4
    UseOwnerScale=true
    Yaw / Pitch / Roll=<0 0.25 0>
   )
   &ShapeLocation(
    MaskSpawnGroup=true
    Override Seed=false
    Rotation Mode=::NewEnumerator2
    Shape Origin=&AddVectorToPosition(
     Position=$StackContext.Position
     Vector=&Multiply_Vector(
      A=<207.3846 550.5953 -570.5737>
      B=&MakeVector(
       X=1
       Y=&UniformAOrBFloat(A=-1, Evaluation Type=::NewEnumerator0, Recalculate Random Each Loop=true, Override Seed=false)
       Z=1
      )
     )
    )
    Spawn Group Mask=1
    Sphere Radius=148.652
    UseOwnerScale=true
    Yaw / Pitch / Roll=<0.25 0.25 0>
   ) {module=ShapeLocation001}
   /SetVariables(Particles.SpriteRotation:NiagaraFloat=&RandomRangeFloat(Minimum=-0.2293, Maximum=0.3106)) {isSetParameters=true, module=SetVariables_#3}
   &AddVelocity(Velocity=$StackContext.ShapeLocation.ShapeVector, Velocity Speed Scale=0)
   /SetVariables(Particles.SpriteSize=&Multiply_Vector2D_ByFloat(
     Float=&Lerp_Float(A=107.586, B=190.6381, Alpha=&MakeFloatFromInt(Int=$Particles.SpawnGroup))
     Vector2D=$Particles.SpriteSize
    )) {isSetParameters=true, module=SetVariables_#4}
   &ApplyInitialForces()
  ]
  particleUpdateScript [
   &Drag(Drag=2, Use Linear Drag=true, Use Rotational Drag=true)
   &VortexForce(
    Origin Pull Amount=76.4139
    UseOriginOffset=false
    Vortex Axis=<1 0 0>
    Vortex Force Amount=&Lerp_Float(
     A=&RandomRangeFloat(Minimum=93.3802, Maximum=178.6482)
     B=0
     Alpha=&MakeFloatFromInt(Int=$Particles.SpawnGroup)
    )
   )
   &GravityForce(Gravity=&Lerp_Vector(A=<0 0 -85.0702>, B=<0 0 -226.7489>, Alpha=&MakeFloatFromInt(Int=$Particles.SpawnGroup)))
   &AccelerationForce(Acceleration=<300 0 0>, Coordinate Space=::Local)
   &SolveForcesAndVelocity(Clamp Velocity=false, Limit Acceleration=false)
   &ParticleState()
   &ScaleSpriteSize(
    Uniform Curve Scale=5
    Uniform Curve Sprite Scale=DataInterface() {propertyValues {Curve {keys [{arriveTangent=1.2245, interpMode=RCIM_Cubic, leaveTangent=1.2245, tangentMode=RCTM_User, time=0, value=1.5}, <at 1 2 RCIM_Cubic>]}}}
   )
   /SetVariables(Particles.SubImageIndex=&FloatFromCurve(FloatCurve=DataInterface() {propertyValues {Curve {keys [<at 0 0>,<at 1 1>]}}}, Scale Curve=16)) {isSetParameters=true, module=SetVariables_#5}
   &ScaleColor(
    Scale Alpha=&FloatFromCurve(FloatCurve=DataInterface() {propertyValues {Curve {keys [
         <at 0 0 RCIM_Cubic>
         {arriveTangent=0.8929, interpMode=RCIM_Cubic, leaveTangent=0.8929, time=0.2, value=1}
         {arriveTangent=-1.25, interpMode=RCIM_Cubic, leaveTangent=-1.25, time=0.7, value=0.625}
         <at 1 0 RCIM_Cubic>
        ]}}})
    Scale RGB=&VectorFromFloat(Value=&FloatFromCurve(FloatCurve=DataInterface() {propertyValues {Curve {keys [
          {arriveTangent=-8.0933, interpMode=RCIM_Cubic, leaveTangent=-8.0933, tangentMode=RCTM_User, time=0, value=1}
          {arriveTangent=-0.9139, interpMode=RCIM_Cubic, leaveTangent=-0.9139, tangentMode=RCTM_User, time=0.2322, value=0.0558}
          {arriveTangent=-0.2701, interpMode=RCIM_Cubic, leaveTangent=-0.2701, time=0.2734, value=0.0702}
          {arriveTangent=-0.3098, interpMode=RCIM_Cubic, leaveTangent=-0.3098, time=0.3339, value=0.0283}
          <at 0.5 0>
         ]}}}))
    ScaleA=true
    ScaleRGB=true
   )
   &DynamicMaterialParameters(
    Index 0 Param 4=&FloatFromCurve(FloatCurve=DataInterface() {propertyValues {Curve {keys [<at 0 1>,<at 1 0>]}}}, Scale Curve=2)
    Module.Index0Param0Write=true
    Module.Index0Param1Write=true
    Module.Index0Param2Write=true
    Module.Index0Param3Write=true
   )
   &CameraOffset(Camera Offset Amount=&FloatFromCurve(FloatCurve=DataInterface() {propertyValues {Curve {keys [<at 0 25>,<at 1 75>]}}}))
   &self:MatchScale() {module=MatchScale}
  ]
  rendererClasses [&NiagaraSpriteRendererProperties]
  renderers [{rendererClass=&NiagaraSpriteRendererProperties}]
 }
 SparksLayer1 {
  simTarget=GPUComputeSim
  emitterUpdateScript [
   &EmitterState(Life Cycle Mode=::NewEnumerator1, Loop Duration=1, UseLoopDelay=false)
   &SpawnRate(SpawnRate=20000, Use Spawn Probability=false)
  ]
  particleSpawnScript [
   &InitializeParticle(
    Lifetime=&FloatFromCurve(
     FloatCurve=DataInterface() {propertyValues {Curve {keys [
         {interpMode=RCIM_Cubic, tangentMode=RCTM_User, time=0, value=3}
         {interpMode=RCIM_Cubic, tangentMode=RCTM_User, time=0.5, value=7}
         {tangentMode=RCTM_User, time=1, value=3}
        ]}}}
     CurveIndex=&Frac_Float(Float=$Engine.Time)
    )
    Mass Mode=::NewEnumerator1
    Sprite Size=&Multiply_Vector2(
     A=&Vector2DFromFloat(Value=&UniformRangedFloat(Minimum=0.4, Maximum=0.6))
     B=<2 6>
    )
    Sprite Size Mode=::NewEnumerator4
    UsePositionOffset=false
    Write Material Random=true
   ) {module=InitializeParticle001}
   &CylinderLocation(
    Cylinder Height=1
    Cylinder Radius=$System.SparkRadius
    Hemisphere X=true
    Hemisphere Y=true
    MaskSpawnGroup=false
    Module.Enable Lathe Profile=false
    Module.Override Randomness=false
    Module.Surface Only=true
    Orientation Axis:ENiagaraOrientationAxis=::XAxis
    Use Endcaps In Surface Only Mode=false
   )
   &self:InitialPosCHange(Angle:NiagaraFloat=&Multiply_Float(A=180, B=$Engine.Time)) {module=InitialPosCHange}
   /SetVariables(Transient.Position2:NiagaraPosition=$Particles.Position) {isSetParameters=true}
   &CylinderLocation(
    Cylinder Height=1
    Cylinder Origin:Vector3f=/Niagara/DynamicInputs/Helpers/SimulationPosition()
    Cylinder Radius=$System.SparkRadius
    Lathe Profile=DataInterface() {propertyValues {Curve {keys [<at 0 0.451 RCIM_Cubic>, {arriveTangent=0.049, interpMode=RCIM_Cubic, leaveTangent=0.049, time=0.5, value=1}, <at 1 0.5 RCIM_Cubic>]}}}
    MaskSpawnGroup=false
    Module.Enable Lathe Profile=false
    Module.Override Randomness=false
    Module.Surface Only=true
    Orientation Axis:ENiagaraOrientationAxis=::XAxis
    Use Endcaps In Surface Only Mode=false
   ) {module=CylinderLocation001}
   &self:ChoosePos(Lerp=0.3) {module=ChoosePos}
   /Niagara/Modules/Update/Forces/CurlNoiseForce(Noise Frequency=0.005, Noise Quality / Cost=::NewEnumerator3, Noise Strength=50, Pan Noise=true, Pan Noise Field=<1 0 0>) {module=CurlNoiseForce001}
   &PointForce(Force Strength=-250, Module.Use Falloff=false, Module.UseLinearForce=false) {module=PointForce001}
   &ApplyInitialForces()
   /SetVariables(Particles.Position=&Add_Vector(A=Linked() {linkedVariable {name=Particles.Position, type=NiagaraPosition}}, B=$Transient.PhysicsForce)) {isSetParameters=true}
  ]
  particleUpdateScript [
   &ParticleState()
   &CurlNoiseForce(Noise Frequency=0.0015, Noise Quality / Cost=::NewEnumerator3, Noise Strength=7.5, Pan Noise=true, Pan Noise Field=<0.5 0 0>)
   &self:ResampleFroce(Intensity=3) {module=ResampleFroce}
   &VortexForce(Influence Falloff Exponent=1.5, Influence Falloff Radius=1500, Use Influence Falloff=true, Vortex Axis=<1 0 0>, Vortex Force Amount=800)
   &PointForce(Force Falloff Distance=1000, Force Strength=400, Module.Use Falloff=true, Module.UseLinearForce=false)
   &Drag(
    Drag=&FloatFromCurve(
     FloatCurve=DataInterface() {propertyValues {Curve {keys [
         <at -0.0101 2.7154>
         {arriveTangent=-0.3026, interpMode=RCIM_Cubic, leaveTangent=-0.3026, tangentMode=RCTM_User, time=0.0352, value=1.1374}
         {arriveTangent=1.8448, interpMode=RCIM_Cubic, leaveTangent=1.8448, time=0.5183, value=1.5087}
         {arriveTangent=3.0746, interpMode=RCIM_Cubic, leaveTangent=3.0746, tangentMode=RCTM_User, time=0.8151, value=2.5761}
        ]}}}
     CurveIndex=&Divide_Float(A=$Particles.Age, B=6.5)
    )
    Use Linear Drag=true
    Use Rotational Drag=false
   )
   &SolveForcesAndVelocity(Clamp Velocity=false, Limit Acceleration=false)
   &ScaleColor(
    Scale Mode=::NewEnumerator1
    Scale RGBA=&Vector4FromCurve(Vector4Curve=DataInterface() {propertyValues {
       AlphaCurve {keys [
         <at 0.0125 0 RCIM_Cubic>
         {arriveTangent=0.5858, interpMode=RCIM_Cubic, leaveTangent=0.5858, time=0.0846, value=0.5}
         {arriveTangent=-0.5462, interpMode=RCIM_Cubic, leaveTangent=-0.5462, time=0.8661, value=0.5}
         <at 1 0 RCIM_Cubic>
        ]}
       BlueCurve {keys [<at 0.0887 2.7858>,<at 0.2437 0.2048>,<at 0.6292 0>]}
       GreenCurve {keys [<at 0.0887 0.6772>,<at 0.2437 0.0263>,<at 0.6292 0>]}
       RedCurve {keys [<at 0.0887 10>,<at 0.2437 1>,<at 0.6292 0>]}
      }})
    ScaleRGBA=true
   )
   &ScaleSpriteSize(Scale Sprite Size Mode=::NewEnumerator0, Uniform Scale Factor=6) {enabled=false}
   &ScaleSpriteSizeBySpeed(Max Scale Factor=<6 10>, Min Scale Factor=<6 3>, Velocity Threshold=300)
   &self:MatchScale() {module=MatchScale}
   &CameraOffset(Camera Offset Amount=&FloatFromCurve(
     FloatCurve=DataInterface() {propertyValues {Curve {keys [{tangentMode=RCTM_User, time=0, value=0}, {tangentMode=RCTM_User, time=1, value=2}]}}}
     Scale Curve=25
    ))
  ]
  rendererClasses [&NiagaraSpriteRendererProperties]
  renderers [{rendererClass=&NiagaraSpriteRendererProperties}]
 }
}
summary {
 systemName=NE_GateVFX
 emitters [
  {emitterName=SparksLayer1, simTarget=GPUComputeSim, rendererClasses [&NiagaraSpriteRendererProperties]}
  {emitterName=Center, rendererClasses [&NiagaraSpriteRendererProperties]}
  {emitterName=Smoke, simTarget=GPUComputeSim, rendererClasses [&NiagaraSpriteRendererProperties]}
  {emitterName=Minimal, rendererClasses [&NiagaraMeshRendererProperties]}
  {emitterName=MeshRibbon, rendererClasses [&NiagaraMeshRendererProperties]}
  {emitterName=MeshRibbon001, rendererClasses [&NiagaraMeshRendererProperties]}
  {emitterName=SmokGround, simTarget=GPUComputeSim, rendererClasses [&NiagaraSpriteRendererProperties]}
 ]
 userVariables [User.ChaosDestructionData {
   type=NiagaraDataInterfaceChaosDestruction
   description=""
   defaultValue {
    struct=NiagaraExt_VariableValue_DataInterface
    value {
     dataInterface=&self:NiagaraDataInterfaceChaosDestruction_0
     dataInterfaceClass=/Script/ChaosNiagara.NiagaraDataInterfaceChaosDestruction
    }
   }
  }]
}
